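# Unlit, two-sided translucent beam-trail material for the SuperDingo consumable's air trail effect.
# Color blends particle color toward an HDR-remapped Light Color along a sine-shaped center band, whose contrast varies along U to change the band's width.
# Opacity multiplies particle alpha, panning soft smoke noise on the second UV channel, and an adjustable fade toward the U=1 end.
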Material /Figment_S10/Athena/Items/Consumables/SuperDingo/FX/M_SuperDingo_AirTrail
graph {
 properties {
  bUsedWithBeamTrails=true
  blendMode=BLEND_Translucent
  pixelDepthOffsetMode=PDOM_Legacy
  shadingModel=MSM_Unlit
  twoSided=true
  usedShadingModels=Unlit
 }
 nodes [
  n1 = VectorParameter() {parameterName=Light Color, defaultValue=<rgba 2 0.7028 0.0043 1>}
  n2 = ComponentMask(None=TextureCoordinate()) {g=true}
  n3 = ScalarParameter() {defaultValue=25, parameterName=MidFade ContrastA}
  n4 = ScalarParameter() {defaultValue=200, parameterName=MidFade ContrastB}
  n5 = Power(Base=ComponentMask(None=TextureCoordinate()) {r=true}, Exp=ScalarParameter() {defaultValue=2, parameterName=MidSideFadeContrast})
  LinearInterpolate(
   A=ParticleColor().RGB
   B=MaterialFunctionCall(LDR Input=Clamp(None=@n1.RGB), HDR Tint=@n1.RGB) {materialFunction=&DeriveHDRfromLDR}.RemappedRange
   Alpha=Power(Base=Sine(None=@n2) {period=2}, Exp=LinearInterpolate(A=@n3, B=@n4, Alpha=@n5))
  )
  n6 = TextureCoordinate() {coordinateIndex=1}
  n7 = AppendVector(A=ScalarParameter() {defaultValue=1.5, parameterName=NoiseUtliing}, B=ScalarParameter() {defaultValue=1, parameterName=NoiseVtliing})
  n8 = ScalarParameter() {defaultValue=0.5, parameterName=PanSpeedU}
  n9 = ScalarParameter() {parameterName=PanSpeedV}
  n10 = TextureSample(UVs=Panner(Coordinate=Multiply(A=@n6, B=@n7), Speed=AppendVector(A=@n8, B=@n9)) {speedX=0.5}) {samplerType=SAMPLERTYPE_Grayscale, texture=&T_Smoke_Trail_Soft}
  n11 = ComponentMask(None=TextureCoordinate()) {r=true}
  Multiply(
   A=Multiply(A=ParticleColor().A, B=Clamp(None=@n10.G))
   B=Power(Base=OneMinus(None=@n11), Exp=ScalarParameter() {defaultValue=1, parameterName=End Fade Contrast})
  )
 ]
}
